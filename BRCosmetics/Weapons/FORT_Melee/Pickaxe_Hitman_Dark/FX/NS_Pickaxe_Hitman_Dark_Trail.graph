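# Niagara swing trail for the Hitman Dark pickaxe, combining air and colored anim-trail ribbons, socket-spawned spinning sprites and a looping pulsed light.
# Ribbons and sprites follow User.SkeletalMesh, and time of day scales the colored ribbon's alpha.
# User.ColorSwitcher blends teal-violet toward greys, with random accent tints on sprites, and lowers the light's spawn probability to zero.
# Alternate color ribbon and sprite emitters are present but disabled.
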NiagaraSystem /BRCosmetics/Weapons/FORT_Melee/Pickaxe_Hitman_Dark/FX/NS_Pickaxe_Hitman_Dark_Trail
dependencies {
 usedDataInterfaces [&NiagaraDataInterfaceColorCurve, &NiagaraDataInterfaceCurve, &NiagaraDataInterfaceArrayColor]
 usedDynamicInputs []
 usedModules [&SystemState]
}
emitters {
 AnimTrail_Air {
  emitterUpdateScript [
   &EmitterState()
   &FN_SpawnAnimTrail()
  ]
  particleSpawnScript [
   &FN_InitializeAnimTrailBoneIndecies(First Bone or Socket Index=1, Second Bone or Socket Index=2)
   &FN_InitializeAnimTrail(
    Color=&LerpLinearColors(StartColor=<rgba 0.4 0.4 0.4 2.2>, EndColor=<rgba 1 1 1 0.4>, LerpFactor=$User.ColorSwitcher)
    Lifetime=0.125
   )
  ]
  particleUpdateScript [
   &SolveForcesAndVelocity(Clamp Velocity=false, Limit Acceleration=false)
   &FN_UpdateAnimTrail(Locked Particle RibbonWidth=1, SkeletalMesh=$User.SkeletalMesh) {module=FN_UpdateAnimTrail001}
   &ScaleColor(
    Scale Alpha=&FloatFromCurve(
     FloatCurve=DataInterface() {propertyValues {Curve {keys [
         {arriveTangent=-2.003, interpMode=RCIM_Cubic, leaveTangent=-2.003, tangentMode=RCTM_User, time=0.25, value=0.95}
         {interpMode=RCIM_Cubic, tangentMode=RCTM_User, time=1, value=0}
        ]}}}
     Scale Curve=0.5
    )
    ScaleA=true
    ScaleRGB=true
   )
  ]
  rendererClasses [&NiagaraRibbonRendererProperties]
  renderers [{rendererClass=&NiagaraRibbonRendererProperties}]
 }
 AnimTrail_Color {
  emitterUpdateScript [
   &EmitterState()
   &FN_SpawnAnimTrail_v2()
  ]
  particleSpawnScript [
   /Game/Effects/Niagara/Modules/AnimTrail/FN_InitializeAnimTrail/FN_InitializeAnimTrailBoneIndecies(First Bone or Socket Index=1)
   /Game/Effects/Niagara/Modules/AnimTrail/FN_InitializeAnimTrail/FN_InitializeAnimTrail_v3/FN_InitializeAnimTrail(Lifetime=0.222, Write Material Random=false)
  ]
  particleUpdateScript [
   /Game/Effects/Niagara/Modules/AnimTrail/FN_UpdateAnimTrail/FN_UpdateAnimTrail_v3/FN_UpdateAnimTrail(Locked Particle RibbonWidth:NiagaraFloat=$Particles.RibbonWidth)
   &Color(Color=&LerpLinearColors(
     StartColor=&ColorFromCurve(ColorCurve=DataInterface() {propertyValues {
        AlphaCurve {keys [<at 0 1>,<at 1 1>]}
        BlueCurve {keys [<at 0.0604 0.4635>, {arriveTangent=-0.0001, leaveTangent=-0.0001, time=0.298, value=1}]}
        GreenCurve {keys [{arriveTangent=-1720.738, time=0.0604, value=0.3624}, {arriveTangent=-0.4891, leaveTangent=-0.4891, time=0.298, value=0.1221}]}
        RedCurve {keys [{arriveTangent=30.0741, time=0.0604, value=0.0628}, {arriveTangent=52.7832, leaveTangent=52.7832, time=0.298, value=0.2502}]}
       }})
     EndColor=&LerpLinearColors(
      StartColor=&ColorFromCurve(ColorCurve=DataInterface() {propertyValues {
         AlphaCurve {keys [<at 0 1>,<at 1 1>]}
         BlueCurve {keys [<at 0.0604 0.6038>, {arriveTangent=-0.0001, leaveTangent=-0.0001, time=0.2242, value=0.5395}]}
         GreenCurve {keys [{arriveTangent=-1720.738, time=0.0604, value=0.5483}, {arriveTangent=-0.4891, leaveTangent=-0.4891, time=0.2242, value=0.5029}]}
         RedCurve {keys [{arriveTangent=30.0741, time=0.0604, value=0.4831}, {arriveTangent=52.7832, leaveTangent=52.7832, time=0.2242, value=0.4233}]}
        }})
      EndColor=&ColorFromCurve(ColorCurve=DataInterface() {propertyValues {
         AlphaCurve {keys [<at 0 1>,<at 1 1>]}
         BlueCurve {keys [<at 0.0604 0.5133>, {arriveTangent=-0.0001, leaveTangent=-0.0001, time=0.2242, value=0.5395}]}
         GreenCurve {keys [{arriveTangent=-1720.738, time=0.0604, value=0.6038}, {arriveTangent=-0.4891, leaveTangent=-0.4891, time=0.2242, value=0.5029}]}
         RedCurve {keys [{arriveTangent=30.0741, time=0.0604, value=0.4831}, {arriveTangent=52.7832, leaveTangent=52.7832, time=0.2242, value=0.4233}]}
        }})
      LerpFactor=&MakeCustomFloatFromBool(Bool=&RandomBool(), True Float=1, False Float=0)
     )
     LerpFactor=$User.ColorSwitcher
    )) {module=Color001}
   &ScaleColor(
    Color Value To Scale=$Particles.Color
    Scale Alpha=&Multiply_Float(
     A=&FloatFromCurve(FloatCurve=DataInterface() {propertyValues {Curve {keys [{arriveTangent=19.9617, interpMode=RCIM_Cubic, tangentMode=RCTM_User, time=0, value=1}, <at 1 1>]}}})
     B=&FNTimeOfDayScalar(Night Scale=0.3, Morning Scale=0.5, Evening Scale=0.5)
    )
    Scale RGB=&Multiply_VectorByFloat(Vector=<1 1 1>, Float=&Lerp_Float(A=0.7, B=0.5, Alpha=$User.ColorSwitcher))
    ScaleA=true
    ScaleRGB=true
   )
  ]
  rendererClasses [&NiagaraRibbonRendererProperties]
  renderers [{rendererClass=&NiagaraRibbonRendererProperties}]
 }
 AnimTrail_Color001 {
  bEnabled=false
  emitterUpdateScript [
   &EmitterState()
   &FN_SpawnAnimTrail_v2()
  ]
  particleSpawnScript [
   &FN_InitializeAnimTrailBoneIndecies(First Bone or Socket Index=1)
   &FN_InitializeAnimTrail(Lifetime=0.222, Write Material Random=false)
  ]
  particleUpdateScript [
   &FN_UpdateAnimTrail(Locked Particle RibbonWidth:NiagaraFloat=$Particles.RibbonWidth)
   &Color(Color=&LerpLinearColors(
     StartColor=&ColorFromCurve(ColorCurve=DataInterface() {propertyValues {
        AlphaCurve {keys [<at 0 1>,<at 1 1>]}
        BlueCurve {keys [<at 0.0604 0.4635>, {arriveTangent=-0.0001, leaveTangent=-0.0001, time=0.298, value=1}]}
        GreenCurve {keys [{arriveTangent=-1720.738, time=0.0604, value=0.3624}, {arriveTangent=-0.4891, leaveTangent=-0.4891, time=0.298, value=0.1221}]}
        RedCurve {keys [{arriveTangent=30.0741, time=0.0604, value=0.0628}, {arriveTangent=52.7832, leaveTangent=52.7832, time=0.298, value=0.2502}]}
       }})
     EndColor=&LerpLinearColors(
      StartColor=&ColorFromCurve(ColorCurve=DataInterface() {propertyValues {
         AlphaCurve {keys [<at 0 1>,<at 1 1>]}
         BlueCurve {keys [<at 0.0604 0.6038>, {arriveTangent=-0.0001, leaveTangent=-0.0001, time=0.2242, value=0.5395}]}
         GreenCurve {keys [{arriveTangent=-1720.738, time=0.0604, value=0.6038}, {arriveTangent=-0.4891, leaveTangent=-0.4891, time=0.2242, value=0.5395}]}
         RedCurve {keys [{arriveTangent=30.0741, time=0.0604, value=0.6038}, {arriveTangent=52.7832, leaveTangent=52.7832, time=0.2242, value=0.5395}]}
        }})
      EndColor=&ColorFromCurve(
       ColorCurve=DataInterface() {propertyValues {
         AlphaCurve {keys [<at 0 1 RCIM_Cubic>,<at 1 1 RCIM_Cubic>]}
         BlueCurve {keys [<at 0.038 0.032 RCIM_Cubic>,<at 0.3695 0.5583>,<at 0.6799 0.0953>,<at 0.948 0.0395 RCIM_Cubic>]}
         GreenCurve {keys [<at 0.038 0 RCIM_Cubic>,<at 0.3695 0.3613>,<at 0.6799 0.6514>,<at 0.948 0.6038 RCIM_Cubic>]}
         RedCurve {keys [<at 0.038 0.851 RCIM_Cubic>,<at 0.3695 0.0024>,<at 0.6799 0.0887>,<at 0.948 0.7835 RCIM_Cubic>]}
        }}
       CurveIndex=&RandomRangeFloat(Minimum=0, Maximum=1)
      )
      LerpFactor=0.1
     )
     LerpFactor=$User.ColorSwitcher
    )) {module=Color001}
   &ScaleColor(
    Color Value To Scale=$Particles.Color
    Scale Alpha=&Multiply_Float(
     A=&FloatFromCurve(FloatCurve=DataInterface() {propertyValues {Curve {keys [{arriveTangent=19.9617, interpMode=RCIM_Cubic, tangentMode=RCTM_User, time=0, value=1}, <at 1 1>]}}})
     B=&FNTimeOfDayScalar(Night Scale=0.3, Morning Scale=0.5, Evening Scale=0.5)
    )
    Scale RGB=&Multiply_VectorByFloat(Vector=<1 1 1>, Float=&Lerp_Float(A=0.7, B=0.5, Alpha=$User.ColorSwitcher))
    ScaleA=true
    ScaleRGB=true
   )
  ]
  rendererClasses [&NiagaraRibbonRendererProperties]
  renderers [{rendererClass=&NiagaraRibbonRendererProperties}]
 }
 Light {
  emitterUpdateScript [
   &EmitterState(Life Cycle Mode=::NewEnumerator1, Loop Behavior=::NewEnumerator1, Loop Delay=0.05, Loop Duration=0.3, UseLoopDelay=true)
   &SpawnBurst_Instantaneous(
    Spawn Probability=&Lerp_Float(A=1, B=0, Alpha=$User.ColorSwitcher)
    Use Spawn Probability=true
    UseLoopCountLimit=false
   )
  ]
  particleSpawnScript [&InitializeParticle(
    Color=&LerpLinearColors(StartColor=<rgba 0.2652 0.1209 1 1>, EndColor=<rgba 0.6562 0.6562 0.6562 1>, LerpFactor=$User.ColorSwitcher)
    Lifetime=0.3
    UsePositionOffset=false
    Write Material Random=true
   )]
  particleUpdateScript [
   &ParticleState()
   &ScaleColor(
    Scale Alpha=&FloatFromCurve(FloatCurve=DataInterface() {propertyValues {Curve {keys [
         {interpMode=RCIM_Cubic, tangentMode=RCTM_User, time=0, value=0}
         {interpMode=RCIM_Cubic, tangentMode=RCTM_User, time=0.5, value=1}
         {tangentMode=RCTM_User, time=1, value=0}
        ]}}})
    ScaleA=true
    ScaleRGB=true
   )
   /SetVariables(
    Particles.Color=&MultiplyLinearColor(
     Linear Color=$Particles.Initial.Color
     Scale Factor=&MakeVector4_FromFloat(Float=&FloatFromCurve(FloatCurve=DataInterface() {propertyValues {Curve {keys [
           {interpMode=RCIM_Cubic, tangentMode=RCTM_User, time=0, value=0}
           {interpMode=RCIM_Cubic, tangentMode=RCTM_User, time=0.5, value=1}
           {tangentMode=RCTM_User, time=1, value=0}
          ]}}}))
    )
    Particles.Position=&Add_Vector(
     A=Linked() {linkedVariable {name=Particles.Position, type=NiagaraPosition}}
     B=/Niagara/DynamicInputs/Transforms/TransformPosition(Position=<0 0 0.7>, Source Space=::Local, Destination Space=::Simulation)
    )
   ) {isSetParameters=true}
  ]
  rendererClasses [&NiagaraLightRendererProperties]
  renderers [{rendererClass=&NiagaraLightRendererProperties, rendererIndex=0}]
 }
 Particle_Trail {
  emitterUpdateScript [
   &EmitterState()
   &SpawnRate(SpawnRate=60, Use Spawn Probability=false)
  ]
  particleSpawnScript [
   &InitializeParticle(
    Color Mode=::NewEnumerator0
    Lifetime=0.25
    Sprite Rotation Mode=::NewEnumerator2
    Sprite Size Mode=::NewEnumerator3
    Uniform Sprite Size=&FloatFromCurve(
     FloatCurve=DataInterface() {propertyValues {Curve {keys [<at 0.45 3 RCIM_Constant>,<at 0.9 5 RCIM_Constant>,<at 1 8>]}}}
     CurveIndex=&RandomRangeFloat(Minimum=0, Maximum=1)
     Scale Curve=1.25
    )
    UsePositionOffset=false
    Write Material Random=true
   )
   &InheritVelocity(Speed Limit=true, UseSpeedThreshold=false)
   &FN_BoneFlow(
    BoneID:NiagaraInt32=&RandomRangeInt(Minimum=0, Maximum=0)
    Height:NiagaraFloat=&RandomRangeFloat(Minimum=0, Maximum=0.5)
    SkeletalMesh:NiagaraDataInterfaceSkeletalMesh=$User.SkeletalMesh
    Use Sockets Instead=true
   )
   /SetVariables(Particles.SubImageIndex=&RandomRangeFloat(Minimum=0, Maximum=4)) {isSetParameters=true, module=SetVariables_#1}
   &ApplyInitialForces()
  ]
  particleUpdateScript [
   &ParticleState()
   &SpriteRotationRate(Rotation Rate=&RandomRangeFloat(Minimum=50, Maximum=500))
   &Drag(Drag=5, Use Linear Drag=true, Use Rotational Drag=false)
   &SolveForcesAndVelocity(Clamp Velocity=false, Limit Acceleration=false)
   &Color(
    Color=&LerpLinearColors(
     StartColor=&ColorFromCurve(ColorCurve=DataInterface() {propertyValues {
        AlphaCurve {keys [<at 0.5033 1>,<at 1 1>]}
        BlueCurve {keys [<at 0.0347 1>, {arriveTangent=-0.0001, leaveTangent=-0.0001, time=0.1659, value=1}]}
        GreenCurve {keys [{arriveTangent=-1720.738, time=0.0347, value=1}, {arriveTangent=-0.4891, leaveTangent=-0.4891, time=0.1659, value=0.1221}]}
        RedCurve {keys [{arriveTangent=30.0741, time=0.0347, value=0}, {arriveTangent=52.7832, leaveTangent=52.7832, time=0.1659, value=0.2502}]}
       }})
     EndColor=&LerpLinearColors(
      StartColor=&ColorFromCurve(ColorCurve=DataInterface() {propertyValues {
         AlphaCurve {keys [<at 0.4959 1>,<at 1 1>]}
         BlueCurve {keys [<at 0.0604 0.6038>, {arriveTangent=-0.0001, leaveTangent=-0.0001, time=0.298, value=0.5395}]}
         GreenCurve {keys [{arriveTangent=-1720.738, time=0.0604, value=0.6038}, {arriveTangent=-0.4891, leaveTangent=-0.4891, time=0.298, value=0.5395}]}
         RedCurve {keys [{arriveTangent=30.0741, time=0.0604, value=0.6038}, {arriveTangent=52.7832, leaveTangent=52.7832, time=0.298, value=0.5395}]}
        }})
      EndColor=&MultiplyLinearColor(
       Linear Color=&SelectLinearColorFromArray(
        Array Sampling Mode=::NewEnumerator0
        Direct Array Index=?
        Color Selection Array=DataInterface() {propertyValues {ColorData [<rgba 0.0024 0.3613 0.5583 1>, <rgba 0.0887 0.6514 0.0953 1>, <rgba 0.7835 0.6038 0.0395 1>, <rgba 0.75 0.0625 0.0898 1>]}}
        Random Seed=0
       )
       Scale Factor=&MakeVector4(
        X=1
        Y=1
        Z=1
        W=&FloatFromCurve(FloatCurve=DataInterface() {propertyValues {Curve {keys [<at 0.5 1 RCIM_Cubic>,<at 1 0 RCIM_Cubic>]}}})
       )
      )
      LerpFactor=0.2
     )
     LerpFactor=$User.ColorSwitcher
    )
    Scale Alpha=&RandomRangeFloat(Minimum=0.7, Maximum=1.2)
   )
   &ScaleColor(
    Color Value To Scale=$Particles.Color
    Scale Alpha=&FloatFromCurve(FloatCurve=DataInterface() {propertyValues {Curve {keys [<at 0.5 1>,<at 1 0>]}}})
    ScaleA=true
    ScaleRGB=false
   )
  ]
  rendererClasses [&NiagaraSpriteRendererProperties]
  renderers [{rendererClass=&NiagaraSpriteRendererProperties}]
 }
 Particle_Trail001 {
  bEnabled=false
  emitterUpdateScript [
   &EmitterState()
   &SpawnRate(SpawnRate=60, Use Spawn Probability=false)
  ]
  particleSpawnScript [
   &InitializeParticle(
    Color Mode=::NewEnumerator0
    Lifetime=0.25
    Sprite Rotation Mode=::NewEnumerator2
    Sprite Size Mode=::NewEnumerator3
    Uniform Sprite Size=&FloatFromCurve(
     FloatCurve=DataInterface() {propertyValues {Curve {keys [<at 0.45 3 RCIM_Constant>,<at 0.9 5 RCIM_Constant>,<at 1 8>]}}}
     CurveIndex=&RandomRangeFloat(Minimum=0, Maximum=1)
     Scale Curve=1.25
    )
    UsePositionOffset=false
    Write Material Random=true
   )
   &InheritVelocity(Speed Limit=true, UseSpeedThreshold=false)
   &FN_BoneFlow(
    BoneID:NiagaraInt32=&RandomRangeInt(Minimum=0, Maximum=0)
    Height:NiagaraFloat=&RandomRangeFloat(Minimum=0, Maximum=0.5)
    SkeletalMesh:NiagaraDataInterfaceSkeletalMesh=$User.SkeletalMesh
    Use Sockets Instead=true
   )
   /SetVariables(Particles.SubImageIndex=&RandomRangeFloat(Minimum=0, Maximum=4)) {isSetParameters=true, module=SetVariables_#1}
   &ApplyInitialForces()
  ]
  particleUpdateScript [
   &ParticleState()
   &SpriteRotationRate(Rotation Rate=&RandomRangeFloat(Minimum=50, Maximum=500))
   &Drag(Drag=5, Use Linear Drag=true, Use Rotational Drag=false)
   &SolveForcesAndVelocity(Clamp Velocity=false, Limit Acceleration=false)
   &Color(
    Color=&LerpLinearColors(
     StartColor=&ColorFromCurve(ColorCurve=DataInterface() {propertyValues {
        AlphaCurve {keys [<at 0.5033 1>,<at 1 0>]}
        BlueCurve {keys [<at 0.0347 1>, {arriveTangent=-0.0001, leaveTangent=-0.0001, time=0.1659, value=1}]}
        GreenCurve {keys [{arriveTangent=-1720.738, time=0.0347, value=1}, {arriveTangent=-0.4891, leaveTangent=-0.4891, time=0.1659, value=0.1221}]}
        RedCurve {keys [{arriveTangent=30.0741, time=0.0347, value=0}, {arriveTangent=52.7832, leaveTangent=52.7832, time=0.1659, value=0.2502}]}
       }})
     EndColor=&LerpLinearColors(
      StartColor=&ColorFromCurve(ColorCurve=DataInterface() {propertyValues {
         AlphaCurve {keys [<at 0.4959 1>,<at 1 0>]}
         BlueCurve {keys [<at 0.0604 0.6038>, {arriveTangent=-0.0001, leaveTangent=-0.0001, time=0.298, value=0.5395}]}
         GreenCurve {keys [{arriveTangent=-1720.738, time=0.0604, value=0.6038}, {arriveTangent=-0.4891, leaveTangent=-0.4891, time=0.298, value=0.5395}]}
         RedCurve {keys [{arriveTangent=30.0741, time=0.0604, value=0.6038}, {arriveTangent=52.7832, leaveTangent=52.7832, time=0.298, value=0.5395}]}
        }})
      EndColor=&ColorFromCurve(
       ColorCurve=DataInterface() {propertyValues {
         AlphaCurve {keys [<at 0 1 RCIM_Cubic>,<at 1 1 RCIM_Cubic>]}
         BlueCurve {keys [<at 0.0317 0.0319 RCIM_Cubic>,<at 0.3115 0.5583>,<at 0.6313 0.0953>,<at 0.9311 0.0395 RCIM_Cubic>]}
         GreenCurve {keys [<at 0.0317 0 RCIM_Cubic>,<at 0.3115 0.3613>,<at 0.6313 0.6514>,<at 0.9311 0.6038 RCIM_Cubic>]}
         RedCurve {keys [<at 0.0317 0.8469 RCIM_Cubic>,<at 0.3115 0.0024>,<at 0.6313 0.0887>,<at 0.9311 0.7835 RCIM_Cubic>]}
        }}
       CurveIndex=&RandomRangeFloat(Minimum=0, Maximum=1)
      )
      LerpFactor=0.4
     )
     LerpFactor=$User.ColorSwitcher
    )
    Scale Alpha=&RandomRangeFloat(Minimum=0.7, Maximum=1.2)
   )
  ]
  rendererClasses [&NiagaraSpriteRendererProperties]
  renderers [{rendererClass=&NiagaraSpriteRendererProperties}]
 }
}
summary {
 systemName=NS_Pickaxe_Hitman_Dark_Trail
 emitters [
  {emitterName=AnimTrail_Color, rendererClasses [&NiagaraRibbonRendererProperties]}
  {emitterName=Light, rendererClasses [&NiagaraLightRendererProperties]}
  {emitterName=Particle_Trail, rendererClasses [&NiagaraSpriteRendererProperties]}
  {emitterName=AnimTrail_Air, rendererClasses [&NiagaraRibbonRendererProperties]}
  {bEnabled=false, emitterName=AnimTrail_Color001, rendererClasses [&NiagaraRibbonRendererProperties]}
  {bEnabled=false, emitterName=Particle_Trail001, rendererClasses [&NiagaraSpriteRendererProperties]}
 ]
 userVariables [
  User.ColorSwitcher {type=NiagaraFloat, defaultValue=0, description=""}
  User.EmissiveColorSelection {type=NiagaraFloat, defaultValue=0, description=""}
  User.RibbonIDOffset {type=NiagaraInt32}
  User.SkeletalMesh {
   type=NiagaraDataInterfaceSkeletalMesh
   defaultValue {
    struct=NiagaraExt_VariableValue_DataInterface
    value {dataInterface=&self:NiagaraDataInterfaceSkeletalMesh_0}
   }
  }
 ]
}
